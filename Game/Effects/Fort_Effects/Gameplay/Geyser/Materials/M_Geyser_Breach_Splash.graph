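# Unlit translucent Niagara sprite material for the water splash of a geyser breach effect.
# A panning smoke mask, noise-distorted and shaped by a sine UV gradient, drives a two-color blend tinted by time of day, plus contrast highlights.
# Opacity combines that mask and particle alpha with a swirl dissolve in screen-aligned radial coordinates around the object pivot.
# Dynamic parameters control timing, randomness and emissive; mobile skips distortion and depth fade.
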Material /Game/Effects/Fort_Effects/Gameplay/Geyser/Materials/M_Geyser_Breach_Splash
graph {
 properties {
  bUseMaterialAttributes=true
  bUsedWithNiagaraSprites=true
  blendMode=BLEND_Translucent
  pixelDepthOffsetMode=PDOM_Legacy
  shadingModel=MSM_Unlit
  usedShadingModels=Unlit
 }
 nodes [
  n1 = VectorParameter() {group=Color_Time_Of_Day, parameterName=Color_Morning, sortPriority=0, defaultValue=<rgba 1 0.3343 0.122 1>}
  n2 = VectorParameter() {group=Color_Time_Of_Day, parameterName=Color_Day, sortPriority=1, defaultValue=<rgba 1 0.9054 0.7298 1>}
  n3 = VectorParameter() {group=Color_Time_Of_Day, parameterName=Color_Evening, sortPriority=2, defaultValue=<rgba 1 0.1673 0.1291 1>}
  n4 = VectorParameter() {group=Color_Time_Of_Day, parameterName=Color_Night, sortPriority=3, defaultValue=<rgba 0.229 0.442 1 1>}
  n5 = MaterialFunctionCall(Color_Morning=@n1.RGB, Color_Day=@n2.RGB, Color_Evening=@n3.RGB, Color_Night=@n4.RGB) {materialFunction=&Color_Time_of_Day}
  n6 = ScalarParameter() {defaultValue=0.25, group=Color_Time_Of_Day, parameterName=TimeOfDayDesaturation, sortPriority=4}
  n7 = VectorParameter() {parameterName=UVS, defaultValue=<rgba 1.2 0.8 0.1 -3>}
  n8 = Multiply(A=TextureCoordinate(), B=ComponentMask(None=@n7.RGB) {g=true, r=true})
  n9 = DynamicParameter() {paramNames [Time,Random,Param3,Emissive]}
  n10 = Multiply(A=Add(A=@n8, B=@n9.Random)) {constB=0.75}
  n11 = Subtract(A=ComponentMask(None=TextureSample(UVs=@n10) {samplerType=SAMPLERTYPE_LinearColor, texture=&T_NoisePack_LinearColor}.RGB) {g=true, r=true}) {constB=0.5}
  n12 = ScalarParameter() {defaultValue=0.125, parameterName=UVTurblence}
  n13 = Add(
   A=ShadingPathSwitch(Default=Multiply(A=@n11, B=@n12) {constB=0.125}, Mobile=Constant()) {desc=Migrated from FeatureLevelSwitch}
   B=Panner(Coordinate=@n8, Time=Add(A=@n9.Time, B=@n9.Time), Speed=AppendVector(A=@n7.B, B=@n7.B))
  )
  n14 = TextureCoordinate() {uTiling=0.5, vTiling=0.5}
  n15 = MaterialFunctionCall(Float2=Sine(None=@n14)) {materialFunction=&BreakOutFloat2Components}
  n16 = Multiply(A=@n15.R, B=@n15.R)
  n17 = Multiply(A=@n16, B=Power(Base=ComponentMask(None=@n14) {g=true}) {constExponent=5})
  n18 = Multiply(
   A=Power(
    Base=TextureSampleParameter2D(UVs=@n13) {parameterName=Mask, texture=&T_Inky_Smoke_Tile}.R
    Exp=ScalarParameter() {defaultValue=0.2, parameterName=Exp}
   )
   B=Add(A=Power(Base=@n16), B=Multiply(A=@n17) {constB=100})
  )
  n19 = ShadingPathSwitch(Default=DepthFade(Opacity=@n18, FadeDistance=ScalarParameter() {defaultValue=10000, parameterName=DF_FD}), Mobile=@n18) {desc=Migrated from FeatureLevelSwitch}
  n20 = LinearInterpolate(
   A=VectorParameter() {parameterName=Color2, defaultValue=<rgba 0.3438 0.6339 0.8594 1>}.RGB
   B=VectorParameter() {parameterName=Color1, defaultValue=<rgba 0.3594 0.9647 1.5 1>}.RGB
   Alpha=Multiply(A=@n19) {constB=0.85}
  )
  n21 = Multiply(A=ScalarParameter() {defaultValue=1, parameterName=EmissiveColor}, B=Multiply(A=Desaturation(None=@n5.Result, Fraction=@n6), B=@n20))
  n22 = ScalarParameter() {defaultValue=2, parameterName=Contrast2}
  n23 = Multiply(
   A=Saturate(None=MaterialFunctionCall(In=@n19, Contrast=@n22) {materialFunction=&CheapContrast}.Result)
   B=Multiply(A=ParticleColor().RGB, B=TextureSample(UVs=@n10) {texture=/Game/Global/Textures/Default/Noise/T_Noise_Cloud}.R)
  )
  n24 = LinearInterpolate(
   A=@n19
   B=MaterialFunctionCall(In=@n19, Contrast=ScalarParameter() {defaultValue=1, parameterName=Contrast}) {materialFunction=&CheapContrast}.Result
   Alpha=ScalarParameter() {defaultValue=0.5, parameterName=Blend}
  )
  n25 = ParticleColor()
  n26 = Add(A=@n24, B=@n25.A)
  n27 = Multiply(A=@n24, B=@n25.A) {constB=0.5}
  n28 = MaterialFunctionCall() {materialFunction=&ObjectPivotPoint}
  n29 = MaterialFunctionCall(Vector or UVs=Multiply(A=MaterialFunctionCall!(Center Point World Position=@n28.Object Pivot Location) {materialFunction=/Engine/Functions/Engine_MaterialFunctions02/UVs/WorldSpaceAlignedScreenCoordinates}.Offset and Scaled UVS) {constB=0.0625}) {materialFunction=&VectorToRadialValue}
  n30 = VectorParameter() {group=SplashDissolve, parameterName=SplashDissolveNoiseTex_UV, defaultValue=<rgba 3 0.75 -0.1 -0.4>}
  n31 = Multiply(A=@n29.Radial Coordinates, B=AppendVector(A=@n30.R, B=@n30.R)) {constB=2}
  n32 = DynamicParameter() {parameterIndex=3, paramNames [DTime, "", "", ""]}
  n33 = Multiply(A=AppendVector(A=@n30.B, B=@n30.B), B=@n32.DTime)
  n34 = Multiply(
   A=TextureSample(UVs=Add(A=@n31, B=@n33)) {samplerType=SAMPLERTYPE_LinearGrayscale, texture=/Game/Environments/Hexmap/HexClouds/T_HexmapFog_01_Height02}.R
   B=AppendVector(A=ScalarParameter() {defaultValue=0.125, group=SplashDissolve, parameterName=SplashDissolveNoiseStrength}, B=Constant())
  )
  n35 = VectorParameter() {group=SplashDissolve, parameterName=SplashDissolveTex_UV, defaultValue=<rgba 1 0.25 -0.2 -0.4>}
  n36 = Add(
   A=Multiply(A=@n29.Radial Coordinates, B=AppendVector(A=@n35.R, B=@n35.R)) {constB=2}
   B=Multiply(A=AppendVector(A=@n35.B, B=@n35.B), B=@n32.DTime)
  )
  n37 = Multiply(
   A=Distance(A=ComponentMask(None=@n28.Object Pivot Location) {b=true, g=true, r=true}, B=WorldPosition().XYZ)
   B=ScalarParameter() {defaultValue=0, group=SplashDissolve, parameterName=SplashDistanceScale}
  ) {constB=2}
  n38 = ScalarParameter() {defaultValue=0.1, group=SplashDissolve, parameterName=SplashDissolveValue}
  n39 = Add(
   A=TextureSample(UVs=Add(A=@n34, B=@n36)) {samplerType=SAMPLERTYPE_LinearGrayscale, texture=/Game/Global/Textures/Default/Flowmaps/T_StormSwirlMask}.RGB
   B=Multiply(A=Add(A=@n37, B=@n38), B=ScalarParameter() {defaultValue=3, group=SplashDissolve, parameterName=SplashDissolveWidth})
  )
  n40 = ScalarParameter() {group=SplashDissolve, parameterName=SplashDIssolveContrast}
  SetMaterialAttributes!(
   MaterialAttributes
   Base Color=@n20
   Emissive Color=Multiply(A=Add(A=@n21, B=@n23), B=DynamicParameter() {paramNames [Time,Random,Param3,Emissive]}.Emissive)
   Opacity=Multiply(
    A=StaticSwitchParameter(True=@n26, False=@n27) {parameterName=Dissolve}
    B=MaterialFunctionCall(In=@n39, Contrast=@n40) {materialFunction=&CheapContrast}.Result
   )
   World Position Offset
  ) {attributeSetTypes [
    #1
    #2
    #3
    #4
   ]}
  Time() {bOverride_Period=true, period=60}
 ]
}
