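# Niagara system for a geyser's idle water effect, with a User.IsPrimed bool that switches it to a stronger primed look.
# A pulsing mesh spout core is layered with sprite mist, gloops and thrown drops, the latter two in matching CPU and GPU emitters.
# IsPrimed swaps IdleMist for PrimeMist and raises other emitters' size, speed or rate.
# Spawn radii follow owner scale; distance remaps thin mist, while other emitters distance-cull.
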NiagaraSystem /Game/Effects/Fort_Effects/Gameplay/Geyser/NS_Geyser_Idle
dependencies {
 usedDataInterfaces [
  &NiagaraDataInterfaceCurve
  &NiagaraDataInterfaceColorCurve
  &NiagaraDataInterfaceArrayColor
  &NiagaraDataInterfaceVectorCurve
 ]
 usedDynamicInputs [&TransformVector]
 usedModules [
  &self:SystemSpawnScript.SystemScriptSource.SystemScriptGraph.NiagaraNodeAssignment_2.SetVariables_#1
  &SystemState
 ]
}
emitters {
 IdleMist {
  emitterUpdateScript [
   &EmitterState()
   &SpawnRate(
    SpawnRate=&MakeCustomFloatFromBool(
     Bool=$User.IsPrimed
     True Float=0
     False Float=&Lerp_Float(
      A=2
      B=1.5
      Alpha=&FN_LOD_Distance_Remap(Start Distance=2500, End Distance=8000, Minimum Output=0, Maximum Output=1)
     )
    )
    Use Spawn Probability=false
   )
  ]
  particleSpawnScript [
   &InitializeParticle(
    Color=&MakeCustomLinearColorFromBool(Bool=$User.IsPrimed, True Color=<rgba 1 1 1 0.35>, False Color=<rgba 1 1 1 0.166>)
    Lifetime Max=3
    Lifetime Min=2
    Lifetime Mode=::NewEnumerator1
    Position Offset=<0 0 -81.3413>
    Sprite Size Max=<65 75>
    Sprite Size Min=<40 35>
    Sprite Size Mode=::NewEnumerator2
    UsePositionOffset=true
    Write Material Random=true
   )
   &CylinderLocation(
    Cylinder Height=0
    Cylinder Height Midpoint=0
    Cylinder Radius=&Multiply_Float(A=50, B=&MakeFloatFromVector(Vector=$Engine.Owner.Scale, Channel=::NewEnumerator0))
    MaskSpawnGroup=false
    Offset=<0 0 -100>
    Offset Mode:ENiagara_OffsetMode=::NewEnumerator0
    Override Seed=false
    UseOwnerScale=true
   )
   &AddVelocity(Coordinate Space=::World, Velocity=&RandomRangeVector(Minimum=<0 0 320>, Maximum=<0 0 465>))
   &DynamicMaterialParameters(
    Index 0 Param 1=&RandomRangeFloat(Minimum=1, Maximum=2)
    Index 0 Param 2=1
    Index 0 Param 3=1
    Index 0 Param 4=0.01
    Module.Param0WriteEnabled=true
    Module.Param1WriteEnabled=true
    Module.Param2WriteEnabled=true
    Module.Param3WriteEnabled=true
   )
   /SetVariables(Particles.UVScale:Vector2f=&Lerp_Vector2(
     A=<-1 1>
     B=<1 1>
     Alpha=&MakeFloatFromInt(Int=&RandomRangeInt(Minimum=0, Maximum=1))
    )) {isSetParameters=true, module=SetVariables_#2}
   /SetVariables(Particles.SpriteSize=&Multiply_Vector2D_ByFloat(Float=&Sine(Normalized Angle=$Emitter.Age, Period=1, Scale=0.2, Bias=1), Vector2D=$Particles.SpriteSize)) {isSetParameters=true, module=SetVariables_#3}
  ]
  particleUpdateScript [
   &ParticleState()
   &GravityForce(Gravity=<0 0 -200>)
   &Drag(Drag=&RandomRangeFloat(Minimum=3, Maximum=4), Use Linear Drag=true, Use Rotational Drag=false)
   &SolveForcesAndVelocity(Clamp Velocity=false, Limit Acceleration=false)
   &SubUVAnimation(
    SubUV Animation Mode:ENiagara_SubUVLookupMode=::NewEnumerator1
    SubUV Curve=DataInterface() {propertyValues {Curve {keys [
        {arriveTangent=2.1621, interpMode=RCIM_Cubic, leaveTangent=2.1621, time=0, value=0}
        {arriveTangent=1, interpMode=RCIM_Cubic, leaveTangent=1, time=0.3, value=0.5451}
        {arriveTangent=0.502, interpMode=RCIM_Cubic, leaveTangent=0.502, time=1, value=1}
       ]}}}
   )
   &ScaleSpriteSize(
    Uniform Curve Scale=&Lerp_Float(
     A=2
     B=1
     Alpha=&FN_LOD_Distance_Remap(Start Distance=2500, End Distance=8000, Minimum Output=0, Maximum Output=1)
    )
    Uniform Curve Sprite Scale=DataInterface() {propertyValues {Curve {keys [
        {arriveTangent=12.5217, interpMode=RCIM_Cubic, leaveTangent=12.5217, tangentMode=RCTM_User, time=0, value=0}
        {arriveTangent=3.5, interpMode=RCIM_Cubic, leaveTangent=3.5, time=0.2, value=2.625}
        {interpMode=RCIM_Cubic, leaveTangent=3.5, time=1, value=3.5}
       ]}}}
   )
   &ScaleColor(
    Scale Alpha=&FloatFromCurve(
     FloatCurve=DataInterface() {propertyValues {Curve {keys [
         {arriveTangent=0.002, interpMode=RCIM_Cubic, time=0, value=0}
         {arriveTangent=0.002, interpMode=RCIM_Cubic, leaveTangent=0.002, time=0.1, value=1}
         {interpMode=RCIM_Cubic, leaveTangent=0.002, time=1.0011, value=0.002}
        ]}}}
     Scale Curve=0.5
    )
    Scale RGB=&MakeVectorFromLinearColorRGB(Color=&ColorFromCurve(ColorCurve=DataInterface() {propertyValues {
        AlphaCurve {keys []}
        BlueCurve {keys [<at 0 0.943 RCIM_Cubic>,<at 0.9345 0.943 RCIM_Cubic>]}
        GreenCurve {keys [
          {arriveTangent=0.2881, interpMode=RCIM_Cubic, leaveTangent=0.2881, time=0, value=0.6737}
          {arriveTangent=0.2881, interpMode=RCIM_Cubic, leaveTangent=0.2881, time=0.9345, value=0.943}
         ]}
        RedCurve {keys [
          {arriveTangent=0.7827, interpMode=RCIM_Cubic, leaveTangent=0.7827, time=0, value=0.2115}
          {arriveTangent=0.7827, interpMode=RCIM_Cubic, leaveTangent=0.7827, time=0.9345, value=0.943}
         ]}
       }}))
    ScaleA=true
    ScaleRGB=true
   )
  ]
  rendererClasses [&NiagaraSpriteRendererProperties]
  renderers [{rendererClass=&NiagaraSpriteRendererProperties}]
 }
 Low_Gloops {
  emitterUpdateScript [
   &EmitterState(Enable Distance Culling=true, MaxDistance=6500, Module.bUseMaxDistance=true, Module.bUseMinDistance=false, Scalability Mode=::NewEnumerator1)
   &SpawnRate(
    SpawnRate=&MakeCustomFloatFromBool(Bool=$User.IsPrimed, True Float=12, False Float=7)
    Use Spawn Probability=false
   )
  ]
  particleSpawnScript [
   /Niagara/Modules/Spawn/Initialization/InitializeParticle(
    Color=&SelectLinearColorFromArray(
     Array Sampling Mode=::NewEnumerator0
     Direct Array Index=?
     Color Selection Array=DataInterface() {propertyValues {ColorData [<rgba 0.4204 0.7641 1 1>, <rgba 0.2327 0.6781 1 1>]}}
     Random Seed=0
    )
    Lifetime=&RandomRangeFloat(Minimum=0.5, Maximum=0.75)
    Module.Write Color=true
    Module.Write Lifetime=true
    Module.Write Mass=true
    Module.Write Material Random=true
    Module.Write Position=false
    Module.Write Scale=false
    Module.Write SpriteRotation=true
    Module.Write SpriteSize=true
    Position=&SimulationPosition()
    Sprite Rotation=&UniformRangedFloat(Minimum=0, Maximum=360)
    Sprite Size=&MakeCustomVector2DFromBool(
     Bool=$User.IsPrimed
     True Vector2D=&Vector2DFromFloat(Value=&UniformRangedFloat(Minimum=22, Maximum=30))
     False Vector2D=&Vector2DFromFloat(Value=&UniformRangedFloat(Minimum=16, Maximum=20))
    )
   )
   &SphereLocation(
    Hemisphere Z=true
    MaskSpawnGroup=false
    Offset=<0 0 -75>
    Offset Coordinate Space:ENiagaraCoordinateSpace=::World
    Override Seed=false
    Sphere Coordinate Space=::World
    Sphere Radius=&Multiply_Float(A=30, B=&MakeFloatFromVector(Vector=$Engine.Owner.Scale, Channel=::NewEnumerator0))
    Surface Only=false
    UseNonUniformScale=false
    UseOffset=true
    UseSphereOrientation=false
   )
   &AddVelocityFromPoint(
    Module.Use Falloff Distance=false
    Origin Offset=<0 0 -12>
    Origin Offset Coordinate Space=::World
    Velocity Strength=&UniformRangedFloat(Minimum=700, Maximum=1150)
   ) {enabled=false}
   &AddVelocityInCone(
    Cone Axis=<0 0 1>
    Cone Axis Coordinate Space=::World
    Module.Use Velocity Falloff On Cone Axis=false
    Velocity Strength=&MakeCustomFloatFromBool(
     Bool=$User.IsPrimed
     True Float=&RandomRangeFloat(Minimum=450, Maximum=1250)
     False Float=&RandomRangeFloat(Minimum=325, Maximum=500)
    )
   )
   &AddVelocity(Coordinate Space=::World, Velocity=<0 0 100>) {enabled=false}
  ]
  particleUpdateScript [
   &ParticleState()
   &AccelerationForce(Acceleration=&UniformRangedVector(Minimum=<0 0 -1600>, Maximum=<0 0 -1800>))
   &SolveForcesAndVelocity(Clamp Velocity=false, Limit Acceleration=false)
   &ScaleSpriteSize(Scale Factor=&Vector2DFromFloat(Value=&FloatFromCurve(
      FloatCurve=DataInterface() {propertyValues {Curve {keys [
          {arriveTangent=13.4291, interpMode=RCIM_Cubic, leaveTangent=13.4291, tangentMode=RCTM_User, time=0, value=0.4376}
          {arriveTangent=0.8929, interpMode=RCIM_Cubic, leaveTangent=0.8929, tangentMode=RCTM_User, time=1, value=5}
         ]}}}
      Scale Curve=2
     )))
   &ScaleColor(
    Scale Alpha=&FloatFromCurve(FloatCurve=DataInterface() {propertyValues {Curve {keys [<at 0 0>,<at 0.009 0.9316>,<at 0.8 0.75>,<at 1 0>]}}})
    Scale RGB=&VectorFromCurve(VectorCurve=DataInterface() {propertyValues {XCurve {keys [<at 0 2>,<at 1 0>]}, YCurve {keys [<at 0 2>,<at 1 0>]}, ZCurve {keys [<at 0 2>,<at 1 0>]}}})
    ScaleA=true
    ScaleRGB=true
   )
   &DynamicMaterialParameters(
    Index 0 Param 1=&FloatFromCurve(FloatCurve=DataInterface() {propertyValues {Curve {keys [{arriveTangent=-2.0663, interpMode=RCIM_Cubic, time=0.1, value=1}, {interpMode=RCIM_Cubic, leaveTangent=-1.3063, time=1.0011, value=0}]}}})
    Index 0 Param 2=&FloatFromCurve(FloatCurve=DataInterface() {propertyValues {Curve {keys [<at -0.0086 -1.3001>,<at 0.9945 1.9153>]}}})
    Index 0 Param 3=0.5
    Module.Param0WriteEnabled=true
    Module.Param1WriteEnabled=false
    Module.Param2WriteEnabled=true
    Module.Param3WriteEnabled=true
   ) {module=DynamicMaterialParameters001}
   &SpriteRotationRate(Rotation Rate=&Lerp_Float(A=-150, B=150, Alpha=$Particles.Initial.MaterialRandom))
  ]
  rendererClasses [&NiagaraSpriteRendererProperties]
  renderers [{rendererClass=&NiagaraSpriteRendererProperties}]
 }
 PrimeMist {
  emitterUpdateScript [
   &EmitterState()
   &SpawnRate(
    SpawnRate=&MakeCustomFloatFromBool(
     Bool=$User.IsPrimed
     True Float=&Lerp_Float(
      A=3
      B=1.5
      Alpha=&FN_LOD_Distance_Remap(Start Distance=2500, End Distance=8000, Minimum Output=0, Maximum Output=1)
     )
     False Float=0
    )
    Use Spawn Probability=false
   )
  ]
  particleSpawnScript [
   /Niagara/Modules/Spawn/Initialization/V2/InitializeParticle(
    Color=<rgba 1 1 1 0.35>
    Lifetime Max=3
    Lifetime Min=2
    Lifetime Mode=::NewEnumerator1
    Position Offset=<0 0 -81.3413>
    Sprite Size Max=<100 160>
    Sprite Size Min=<80 100>
    Sprite Size Mode=::NewEnumerator2
    UsePositionOffset=true
    Write Material Random=true
   )
   &CylinderLocation(
    Cylinder Height=0
    Cylinder Height Midpoint=0
    Cylinder Radius=&Multiply_Float(A=50, B=&MakeFloatFromVector(Vector=$Engine.Owner.Scale, Channel=::NewEnumerator0))
    MaskSpawnGroup=false
    Offset=<0 0 -100>
    Offset Mode:ENiagara_OffsetMode=::NewEnumerator0
    Override Seed=false
    UseOwnerScale=true
   )
   &AddVelocity(Coordinate Space=::World, Velocity=&RandomRangeVector(Minimum=<0 0 700>, Maximum=<0 0 1200>))
   &DynamicMaterialParameters(
    Index 0 Param 1=&RandomRangeFloat(Minimum=1, Maximum=2)
    Index 0 Param 2=1
    Index 0 Param 3=1
    Index 0 Param 4=0.01
    Module.Param0WriteEnabled=true
    Module.Param1WriteEnabled=true
    Module.Param2WriteEnabled=true
    Module.Param3WriteEnabled=true
   )
   /SetVariables(Particles.UVScale:Vector2f=&Lerp_Vector2(
     A=<-1 1>
     B=<1 1>
     Alpha=&MakeFloatFromInt(Int=&RandomRangeInt(Minimum=0, Maximum=1))
    )) {isSetParameters=true, module=SetVariables_#2}
   /SetVariables(Particles.SpriteSize=&Multiply_Vector2D_ByFloat(Float=&Sine(Normalized Angle=$Emitter.Age, Period=1, Scale=0.2, Bias=1), Vector2D=$Particles.SpriteSize)) {isSetParameters=true, module=SetVariables_#3}
  ]
  particleUpdateScript [
   &ParticleState()
   &GravityForce(Gravity=<0 0 -200>)
   &Drag(Drag=&RandomRangeFloat(Minimum=3, Maximum=4), Use Linear Drag=true, Use Rotational Drag=false)
   &SolveForcesAndVelocity(Clamp Velocity=false, Limit Acceleration=false)
   &SubUVAnimation(
    SubUV Animation Mode:ENiagara_SubUVLookupMode=::NewEnumerator1
    SubUV Curve=DataInterface() {propertyValues {Curve {keys [
        {arriveTangent=2.1621, interpMode=RCIM_Cubic, leaveTangent=2.1621, time=0, value=0}
        {arriveTangent=1, interpMode=RCIM_Cubic, leaveTangent=1, time=0.3, value=0.5451}
        {arriveTangent=0.502, interpMode=RCIM_Cubic, leaveTangent=0.502, time=1, value=1}
       ]}}}
   )
   &ScaleSpriteSize(
    Uniform Curve Scale=&Lerp_Float(
     A=2
     B=1
     Alpha=&FN_LOD_Distance_Remap(Start Distance=2500, End Distance=8000, Minimum Output=0, Maximum Output=1)
    )
    Uniform Curve Sprite Scale=DataInterface() {propertyValues {Curve {keys [
        {arriveTangent=12.5217, interpMode=RCIM_Cubic, leaveTangent=12.5217, tangentMode=RCTM_User, time=0, value=0}
        {arriveTangent=3.5, interpMode=RCIM_Cubic, leaveTangent=3.5, time=0.2, value=2.625}
        {interpMode=RCIM_Cubic, leaveTangent=3.5, time=1, value=3.5}
       ]}}}
   )
   &ScaleColor(
    Scale Alpha=&FloatFromCurve(
     FloatCurve=DataInterface() {propertyValues {Curve {keys [
         {arriveTangent=0.002, interpMode=RCIM_Cubic, time=0, value=0}
         {arriveTangent=0.002, interpMode=RCIM_Cubic, leaveTangent=0.002, time=0.1, value=1}
         {interpMode=RCIM_Cubic, leaveTangent=0.002, time=1.0011, value=0.002}
        ]}}}
     Scale Curve=0.5
    )
    Scale RGB=&MakeVectorFromLinearColorRGB(Color=&ColorFromCurve(ColorCurve=DataInterface() {propertyValues {
        AlphaCurve {keys []}
        BlueCurve {keys [<at 0 0.943 RCIM_Cubic>,<at 0.9345 0.943 RCIM_Cubic>]}
        GreenCurve {keys [
          {arriveTangent=0.2881, interpMode=RCIM_Cubic, leaveTangent=0.2881, time=0, value=0.6737}
          {arriveTangent=0.2881, interpMode=RCIM_Cubic, leaveTangent=0.2881, time=0.9345, value=0.943}
         ]}
        RedCurve {keys [
          {arriveTangent=0.7827, interpMode=RCIM_Cubic, leaveTangent=0.7827, time=0, value=0.2115}
          {arriveTangent=0.7827, interpMode=RCIM_Cubic, leaveTangent=0.7827, time=0.9345, value=0.943}
         ]}
       }}))
    ScaleA=true
    ScaleRGB=true
   )
  ]
  rendererClasses [&NiagaraSpriteRendererProperties]
  renderers [{rendererClass=&NiagaraSpriteRendererProperties}]
 }
 Secondary_Gloops {
  simTarget=GPUComputeSim
  emitterUpdateScript [
   &EmitterState(Enable Distance Culling=true, MaxDistance=6500, Module.bUseMaxDistance=true, Module.bUseMinDistance=false, Scalability Mode=::NewEnumerator1)
   &SpawnRate(
    SpawnRate=&MakeCustomFloatFromBool(Bool=$User.IsPrimed, True Float=12, False Float=7)
    Use Spawn Probability=false
   )
  ]
  particleSpawnScript [
   /Niagara/Modules/Spawn/Initialization/InitializeParticle(
    Color=&SelectLinearColorFromArray(
     Array Sampling Mode=::NewEnumerator0
     Direct Array Index=?
     Color Selection Array=DataInterface() {propertyValues {ColorData [<rgba 0.4204 0.7641 1 1>, <rgba 0.2327 0.6781 1 1>]}}
     Random Seed=0
    )
    Lifetime=&RandomRangeFloat(Minimum=0.5, Maximum=0.75)
    Module.Write Color=true
    Module.Write Lifetime=true
    Module.Write Mass=true
    Module.Write Material Random=true
    Module.Write Position=false
    Module.Write Scale=false
    Module.Write SpriteRotation=true
    Module.Write SpriteSize=true
    Position=&SimulationPosition()
    Sprite Rotation=&UniformRangedFloat(Minimum=0, Maximum=360)
    Sprite Size=&MakeCustomVector2DFromBool(
     Bool=$User.IsPrimed
     True Vector2D=&Vector2DFromFloat(Value=&UniformRangedFloat(Minimum=22, Maximum=30))
     False Vector2D=&Vector2DFromFloat(Value=&UniformRangedFloat(Minimum=16, Maximum=20))
    )
   )
   &SphereLocation(
    Hemisphere Z=true
    MaskSpawnGroup=false
    Offset=<0 0 -75>
    Offset Coordinate Space:ENiagaraCoordinateSpace=::World
    Override Seed=false
    Sphere Coordinate Space=::World
    Sphere Radius=&Multiply_Float(A=30, B=&MakeFloatFromVector(Vector=$Engine.Owner.Scale, Channel=::NewEnumerator0))
    Surface Only=false
    UseNonUniformScale=false
    UseOffset=true
    UseSphereOrientation=false
   )
   &AddVelocityFromPoint(
    Module.Use Falloff Distance=false
    Origin Offset=<0 0 -12>
    Origin Offset Coordinate Space=::World
    Velocity Strength=&UniformRangedFloat(Minimum=700, Maximum=1150)
   ) {enabled=false}
   &AddVelocityInCone(
    Cone Axis=<0 0 1>
    Cone Axis Coordinate Space=::World
    Module.Use Velocity Falloff On Cone Axis=false
    Velocity Strength=&MakeCustomFloatFromBool(
     Bool=$User.IsPrimed
     True Float=&RandomRangeFloat(Minimum=450, Maximum=1250)
     False Float=&RandomRangeFloat(Minimum=325, Maximum=500)
    )
   )
   &AddVelocity(Coordinate Space=::World, Velocity=<0 0 100>) {enabled=false}
  ]
  particleUpdateScript [
   &ParticleState()
   &AccelerationForce(Acceleration=&UniformRangedVector(Minimum=<0 0 -1600>, Maximum=<0 0 -1800>))
   &SolveForcesAndVelocity(Clamp Velocity=false, Limit Acceleration=false)
   &ScaleSpriteSize(Scale Factor=&Vector2DFromFloat(Value=&FloatFromCurve(
      FloatCurve=DataInterface() {propertyValues {Curve {keys [
          {arriveTangent=13.4291, interpMode=RCIM_Cubic, leaveTangent=13.4291, tangentMode=RCTM_User, time=0, value=0.4376}
          {arriveTangent=0.8929, interpMode=RCIM_Cubic, leaveTangent=0.8929, tangentMode=RCTM_User, time=1, value=5}
         ]}}}
      Scale Curve=2
     )))
   &ScaleColor(
    Scale Alpha=&FloatFromCurve(FloatCurve=DataInterface() {propertyValues {Curve {keys [<at 0 0>,<at 0.009 0.9316>,<at 0.8 0.75>,<at 1 0>]}}})
    Scale RGB=&VectorFromCurve(VectorCurve=DataInterface() {propertyValues {XCurve {keys [<at 0 2>,<at 1 0>]}, YCurve {keys [<at 0 2>,<at 1 0>]}, ZCurve {keys [<at 0 2>,<at 1 0>]}}})
    ScaleA=true
    ScaleRGB=true
   )
   &DynamicMaterialParameters(
    Index 0 Param 1=&FloatFromCurve(FloatCurve=DataInterface() {propertyValues {Curve {keys [{arriveTangent=-2.0663, interpMode=RCIM_Cubic, time=0.1, value=1}, {interpMode=RCIM_Cubic, leaveTangent=-1.3063, time=1.0011, value=0}]}}})
    Index 0 Param 2=&FloatFromCurve(FloatCurve=DataInterface() {propertyValues {Curve {keys [<at -0.0086 -1.3001>,<at 0.9945 1.9153>]}}})
    Index 0 Param 3=0.5
    Module.Param0WriteEnabled=true
    Module.Param1WriteEnabled=false
    Module.Param2WriteEnabled=true
    Module.Param3WriteEnabled=true
   ) {module=DynamicMaterialParameters001}
   &SpriteRotationRate(Rotation Rate=&Lerp_Float(A=-150, B=150, Alpha=$Particles.Initial.MaterialRandom))
  ]
  rendererClasses [&NiagaraSpriteRendererProperties]
  renderers [{rendererClass=&NiagaraSpriteRendererProperties}]
 }
 ThrownDrops002 {
  simTarget=GPUComputeSim
  emitterUpdateScript [
   &EmitterState(Enable Distance Culling=true, MaxDistance=2500, Module.bUseMaxDistance=true, Module.bUseMinDistance=false, Scalability Mode=::NewEnumerator1)
   &SpawnRate(
    SpawnRate=&MakeCustomFloatFromBool(Bool=$User.IsPrimed, True Float=24, False Float=8)
    Use Spawn Probability=false
   )
  ]
  particleSpawnScript [
   &InitializeParticle(
    Color=<rgba 1 1 1 0.4>
    Lifetime=&UniformRangedFloat(Minimum=0.75, Maximum=1.5)
    Mass=&UniformRangedFloat(Minimum=0.75, Maximum=1.75)
    Module.Write Color=true
    Module.Write Lifetime=true
    Module.Write Mass=true
    Module.Write Material Random=true
    Module.Write Position=false
    Module.Write Scale=false
    Module.Write SpriteRotation=false
    Module.Write SpriteSize=true
    Position=&SimulationPosition()
    Sprite Size=&Multiply_Vector2D_ByFloat(Float=0.8, Vector2D=&RandomRangeVector2D(Minimum=<15 25>, Maximum=<15 15>))
   )
   &SphereLocation(
    MaskSpawnGroup=false
    Override Seed=false
    Sphere Radius=&Multiply_Float(A=60, B=&MakeFloatFromVector(Vector=$Engine.Owner.Scale, Channel=::NewEnumerator0))
    Surface Only=false
    UseNonUniformScale=false
    UseOffset=true
    UseSphereOrientation=false
   )
   &AddVelocityInCone(
    Cone Angle=20
    Cone Axis=<0 0 1>
    Module.Use Velocity Falloff On Cone Axis=false
    Velocity Distribution Along Cone Axis=0
    Velocity Strength=&MakeCustomFloatFromBool(
     Bool=$User.IsPrimed
     True Float=&RandomRangeFloat(Minimum=450, Maximum=1250)
     False Float=&RandomRangeFloat(Minimum=325, Maximum=500)
    )
   )
   &AddVelocity(Coordinate Space=::World, Velocity=&RandomRangeVector(Minimum=<0 0 225>, Maximum=<1 1 380>))
  ]
  particleUpdateScript [
   &ParticleState()
   &GravityForce(Gravity=&RandomRangeVector(Minimum=<0 0 -1800>, Maximum=<1 1 -2800>))
   &Drag(Drag=&UniformRangedFloat(Minimum=0.5, Maximum=1), Use Linear Drag=true, Use Rotational Drag=false)
   &SolveForcesAndVelocity(Clamp Velocity=false, Limit Acceleration=false)
   &ScaleSpriteSizeBySpeed(Max Scale Factor=<1 2>, Min Scale Factor=<0.5 0.05>)
   &ScaleColor(
    Scale Mode=::NewEnumerator1
    Scale RGBA=&Vector4FromCurve(Vector4Curve=DataInterface() {propertyValues {
       AlphaCurve {keys [<at 0 0>,<at 0.188 1>,<at 0.6929 1>,<at 1 0>]}
       BlueCurve {keys [<at 0 0.96>]}
       GreenCurve {keys [<at 0 1>]}
       RedCurve {keys [<at 0 0.933>]}
      }})
    ScaleRGBA=true
   )
   &ScaleSpriteSize(Scale Factor=&Vector2DFromFloat(Value=&FloatFromCurve(FloatCurve=DataInterface() {propertyValues {Curve {keys [<at 0 1>,<at 1 0.4>]}}}, Scale Curve=2)))
  ]
  rendererClasses [&NiagaraSpriteRendererProperties]
  renderers [{rendererClass=&NiagaraSpriteRendererProperties}]
 }
 ThrownDrops_Low {
  emitterUpdateScript [
   &EmitterState(Enable Distance Culling=true, MaxDistance=2500, Module.bUseMaxDistance=true, Module.bUseMinDistance=false, Scalability Mode=::NewEnumerator1)
   &SpawnRate(
    SpawnRate=&MakeCustomFloatFromBool(Bool=$User.IsPrimed, True Float=24, False Float=8)
    Use Spawn Probability=false
   )
  ]
  particleSpawnScript [
   &InitializeParticle(
    Color=<rgba 1 1 1 0.4>
    Lifetime=&UniformRangedFloat(Minimum=0.75, Maximum=1.5)
    Mass=&UniformRangedFloat(Minimum=0.75, Maximum=1.75)
    Module.Write Color=true
    Module.Write Lifetime=true
    Module.Write Mass=true
    Module.Write Material Random=true
    Module.Write Position=false
    Module.Write Scale=false
    Module.Write SpriteRotation=false
    Module.Write SpriteSize=true
    Position=&SimulationPosition()
    Sprite Size=&Multiply_Vector2D_ByFloat(Float=0.8, Vector2D=&RandomRangeVector2D(Minimum=<15 25>, Maximum=<15 15>))
   )
   &SphereLocation(
    MaskSpawnGroup=false
    Override Seed=false
    Sphere Radius=&Multiply_Float(A=60, B=&MakeFloatFromVector(Vector=$Engine.Owner.Scale, Channel=::NewEnumerator0))
    Surface Only=false
    UseNonUniformScale=false
    UseOffset=true
    UseSphereOrientation=false
   )
   &AddVelocityInCone(
    Cone Angle=20
    Cone Axis=<0 0 1>
    Module.Use Velocity Falloff On Cone Axis=false
    Velocity Distribution Along Cone Axis=0
    Velocity Strength=&MakeCustomFloatFromBool(
     Bool=$User.IsPrimed
     True Float=&RandomRangeFloat(Minimum=450, Maximum=1250)
     False Float=&RandomRangeFloat(Minimum=325, Maximum=500)
    )
   )
   &AddVelocity(Coordinate Space=::World, Velocity=&RandomRangeVector(Minimum=<0 0 225>, Maximum=<1 1 380>))
  ]
  particleUpdateScript [
   &ParticleState()
   &GravityForce(Gravity=&RandomRangeVector(Minimum=<0 0 -1800>, Maximum=<1 1 -2800>))
   &Drag(Drag=&UniformRangedFloat(Minimum=0.5, Maximum=1), Use Linear Drag=true, Use Rotational Drag=false)
   &SolveForcesAndVelocity(Clamp Velocity=false, Limit Acceleration=false)
   &ScaleSpriteSizeBySpeed(Max Scale Factor=<1 2>, Min Scale Factor=<0.5 0.05>)
   &ScaleColor(
    Scale Mode=::NewEnumerator1
    Scale RGBA=&Vector4FromCurve(Vector4Curve=DataInterface() {propertyValues {
       AlphaCurve {keys [<at 0 0>,<at 0.188 1>,<at 0.6929 1>,<at 1 0>]}
       BlueCurve {keys [<at 0 0.96>]}
       GreenCurve {keys [<at 0 1>]}
       RedCurve {keys [<at 0 0.933>]}
      }})
    ScaleRGBA=true
   )
   &ScaleSpriteSize(Scale Factor=&Vector2DFromFloat(Value=&FloatFromCurve(FloatCurve=DataInterface() {propertyValues {Curve {keys [<at 0 1>,<at 1 0.4>]}}}, Scale Curve=2)))
  ]
  rendererClasses [&NiagaraSpriteRendererProperties]
  renderers [{rendererClass=&NiagaraSpriteRendererProperties}]
 }
 WaterSpoutCore001 {
  emitterUpdateScript [
   &EmitterState(
    Enable Distance Culling=true
    MaxDistance=12000
    Module.bUseMaxDistance=true
    Module.bUseMinDistance=false
    Scalability Mode=::NewEnumerator1
   )
   &SpawnRate(SpawnRate=2, Use Spawn Probability=false)
  ]
  particleSpawnScript [
   /Niagara/Modules/Spawn/Initialization/V2/InitializeParticle(
    Color=<rgba 0.4719 0.8063 1 1>
    Lifetime=1.2
    Mesh Scale=&MakeCustomVectorFromBool(Bool=$User.IsPrimed, True Vector=<8 8 1>, False Vector=<6 6 1>)
    Mesh Scale Mode=::NewEnumerator4
    Position Offset=<0 0 -48.4902>
    UsePositionOffset=true
    Write Material Random=true
   )
   /SetVariables(Particles.WaterPeak:NiagaraFloat=&Sine(Normalized Angle=$Emitter.Age, Period=3, Scale=0.3, Bias=1)) {isSetParameters=true}
   &DynamicMaterialParameters(
    Index 0 Param 1=&MakeCustomFloatFromBool(Bool=$User.IsPrimed, True Float=1, False Float=0.2)
    Index 0 Param 2=&Multiply_Float(A=$Particles.WaterPeak, B=100)
    Index 0 Param 3=1
    Index 0 Param 4=1
    Module.Param0WriteEnabled=true
    Module.Param1WriteEnabled=true
    Module.Param2WriteEnabled=true
    Module.Param3WriteEnabled=true
   )
  ]
  particleUpdateScript [
   &ParticleState()
   &ScaleMeshSize(Scale Factor=&Multiply_Vector(
     A=&VectorFromFloat(Value=&FloatFromCurve(FloatCurve=DataInterface() {propertyValues {Curve {keys [
           {arriveTangent=0.4167, interpMode=RCIM_Cubic, time=0, value=0.25}
           <at 0.5 1 RCIM_Cubic>
           {interpMode=RCIM_Cubic, leaveTangent=-1.875, time=1, value=0.25}
          ]}}}))
     B=<0.5 0.5 0.5>
    ))
   &Color(Color=&ColorFromCurve(ColorCurve=DataInterface() {propertyValues {
       AlphaCurve {keys [
         {arriveTangent=0.0704, interpMode=RCIM_Cubic, leaveTangent=0.0704, tangentMode=RCTM_User, time=0, value=0}
         <at 0.1263 1>
         {arriveTangent=-1.1445, interpMode=RCIM_Cubic, leaveTangent=-1.1445, time=0.8, value=1.125}
         {interpMode=RCIM_Cubic, leaveTangent=-5.625, time=1, value=0}
        ]}
       BlueCurve {keys [<at 0.0139 1>, <at 0.1863 1 RCIM_Cubic>, {interpMode=RCIM_Cubic, leaveTangent=-9.4725, time=1, value=1}]}
       GreenCurve {keys [
         <at 0.0139 0.9369>
         {arriveTangent=-0.2332, interpMode=RCIM_Cubic, leaveTangent=-0.2332, time=0.1863, value=0.9235}
         {interpMode=RCIM_Cubic, leaveTangent=-0.0774, time=1, value=0.707}
        ]}
       RedCurve {keys [
         <at 0.0139 0.1641>
         {arriveTangent=0.4134, interpMode=RCIM_Cubic, leaveTangent=0.4134, time=0.1863, value=0.4043}
         {interpMode=RCIM_Cubic, leaveTangent=0.6018, time=1, value=0.5718}
        ]}
      }}))
  ]
  rendererClasses [&NiagaraMeshRendererProperties]
  renderers [{rendererClass=&NiagaraMeshRendererProperties, rendererIndex=0}]
 }
}
summary {
 systemName=NS_Geyser_Idle
 emitters [
  {emitterName=WaterSpoutCore001, rendererClasses [&NiagaraMeshRendererProperties]}
  {emitterName=Secondary_Gloops, simTarget=GPUComputeSim, rendererClasses [&NiagaraSpriteRendererProperties]}
  {emitterName=ThrownDrops002, simTarget=GPUComputeSim, rendererClasses [&NiagaraSpriteRendererProperties]}
  {emitterName=PrimeMist, rendererClasses [&NiagaraSpriteRendererProperties]}
  {emitterName=IdleMist, rendererClasses [&NiagaraSpriteRendererProperties]}
  {emitterName=Low_Gloops, rendererClasses [&NiagaraSpriteRendererProperties]}
  {emitterName=ThrownDrops_Low, rendererClasses [&NiagaraSpriteRendererProperties]}
 ]
 userVariables [User.IsPrimed {type=NiagaraBool}]
}
